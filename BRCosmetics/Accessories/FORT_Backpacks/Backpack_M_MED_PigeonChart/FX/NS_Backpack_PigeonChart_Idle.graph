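# Niagara idle effect for the PigeonChart backpack cosmetic, built from persistent burst-spawned particles in four dual-renderer emitters.
# Mid and Under place sprites, while Spline_Edgeside and Spline_horn place meshes, mostly by partition-indexed arrays.
# Separate "_Red" attributes define a second color layer, and the spline emitters store smoothed position in Interpolate Dynamic, a trailing-motion input.
# Front-end preview, time of day and platform adjust placement, size or color.
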NiagaraSystem /BRCosmetics/Accessories/FORT_Backpacks/Backpack_M_MED_PigeonChart/FX/NS_Backpack_PigeonChart_Idle
dependencies {
 usedDataInterfaces [
  &NiagaraDataInterfaceArrayFloat3
  &NiagaraDataInterfaceArrayFloat2
  &NiagaraDataInterfacePlatformSet
  &NiagaraDataInterfaceArrayColor
 ]
 usedDynamicInputs []
 usedModules [&SystemState]
}
emitters {
 Mid {
  emitterUpdateScript [
   &EmitterState(
    Inactive Response=::NewEnumerator1
    Life Cycle Mode=::NewEnumerator1
    Loop Behavior=::NewEnumerator1
    Loop Duration Mode=::NewEnumerator1
    UseLoopDelay=false
   )
   &SpawnBurst_Instantaneous(Spawn Count=2, Use Spawn Probability=false, UseLoopCountLimit=false)
  ]
  particleSpawnScript [
   &PartitionParticles()
   &InitializeParticle(
    Color=&MultiplyLinearColor(
     Linear Color=<rgba 0.5 0.2604 0.2604 1.2>
     Scale Factor=&MakeVector4_FromLinearColor(LinearColor=&FNTimeOfDayColor(Night Color=<rgba 0.63 0.6198 1 1>, Morning Color=<rgba 1 1 1 1>, Day Color=<rgba 1 1 1 1>, Evening Color=<rgba 1 1 1 1>))
    )
    Position Offset=&MakeCustomVectorFromBool(
     Bool=$User.bIsFrontEndPreview
     True Vector=&SelectVectorFromArray(
      Direct Array Index=$Particles.PartitionParticles.Partition
      Vector Selection Array=DataInterface() {propertyValues {FloatData [<0 -13 5>, <0 -13 -12>]}}
     )
     False Vector=&SelectVectorFromArray(
      Direct Array Index=$Particles.PartitionParticles.Partition
      Vector Selection Array=DataInterface() {propertyValues {FloatData [<0 -12 5>, <0 -12 -13>]}}
     )
    )
    Sprite Size=&Multiply_Vector2D(
     A=&SelectVector2DFromArray(
      Direct Array Index=$Particles.PartitionParticles.Partition
      Vector 2D Selection Array=DataInterface() {propertyValues {FloatData [<40 30>, <40 30>]}}
     )
     B=&MakeCustomVector2DFromBool(Bool=$User.bIsFrontEndPreview, True Vector2D=<0.8 0.8>, False Vector2D=<1 1>)
    )
    Sprite Size Mode=::NewEnumerator4
    UsePositionOffset=true
    Write Material Random=true
   )
   /SetVariables(Particles.Position_Red:NiagaraPosition=&MakeCustomVectorFromBool(
     Bool=$User.bIsFrontEndPreview
     True Vector=&SelectVectorFromArray(
      Direct Array Index=$Particles.PartitionParticles.Partition
      Vector Selection Array=DataInterface() {propertyValues {FloatData [<0 -13 7>, <0 -13 -5>]}}
     )
     False Vector=&SelectVectorFromArray(
      Direct Array Index=$Particles.PartitionParticles.Partition
      Vector Selection Array=DataInterface() {propertyValues {FloatData [<0 -12 7>, <0 -12 -5>]}}
     )
    )) {isSetParameters=true, module=SetVariables_#1}
   &CameraOffset(Camera Offset Amount=-1)
  ]
  particleUpdateScript [
   &ParticleState(Kill Particles When Lifetime Has Elapsed=false, Let Infinitely Lived Particles Die When Emitter Deactivates=true)
   &ScaleColor(
    Scale Alpha=&FN_FrontEndFloatModifier(Gameplay=0.85, Front End=0.8)
    Scale RGB=&Lerp_Vector(
     A=&MakeVectorFromLinearColorRGB(Color=<rgba 1 1 1 1>)
     B=&MakeVectorFromLinearColorRGB(Color=<rgba 3 0.5312 0.846 1>)
     Alpha=&FN_PickFloatDependOnPlatform(TargetPlatformActive=0, TargetPlatformNoActive=1)
    )
    ScaleA=true
    ScaleRGB=true
   )
   /SetVariables(Particles.Color_Red:LinearColor=&MultiplyLinearColor(
     Linear Color=&SelectLinearColorFromArray(
      Array Sampling Mode=::NewEnumerator1
      Direct Array Index=$Particles.PartitionParticles.Partition
      Color Selection Array=DataInterface() {propertyValues {ColorData [<rgba 0.05 0.0072 0.0068 5>, <rgba 0.06 0.0091 0.0096 3>]}}
      Random Seed=?
     )
     Scale Factor=/Niagara/DynamicInputs/TypeConversions/MakeCustomVector4FromBool(
      Bool:NiagaraBool=$User.bIsFrontEndPreview
      True Vector4:Vector4f=&MakeVector4_FromFloat(Float=&FN_PickFloatDependOnPlatform(TargetPlatformActive=0.8, TargetPlatformNoActive=0.7))
      False Vector4:Vector4f=&MakeVector4_FromFloat(Float=1)
     )
    )) {isSetParameters=true, module=SetVariables_#2}
   /SetVariables(Particles.Size_Red:Vector2f=&Multiply_Vector2D(
     A=&SelectVector2DFromArray(
      Direct Array Index=$Particles.PartitionParticles.Partition
      Vector 2D Selection Array=DataInterface() {propertyValues {FloatData [<45 40>, <55 50>]}}
     )
     B=&MakeCustomVector2DFromBool(Bool=$User.bIsFrontEndPreview, True Vector2D=<0.8 0.8>, False Vector2D=<1 1>)
    )) {isSetParameters=true, module=SetVariables_#3}
  ]
  rendererClasses [&NiagaraSpriteRendererProperties]
  renderers [
   {rendererClass=&NiagaraSpriteRendererProperties}
   {rendererClass=&NiagaraSpriteRendererProperties, rendererIndex=1}
  ]
 }
 Spline_Edgeside {
  emitterUpdateScript [
   &EmitterState(
    Inactive Response=::NewEnumerator1
    Life Cycle Mode=::NewEnumerator1
    Loop Behavior=::NewEnumerator1
    Loop Duration Mode=::NewEnumerator1
    UseLoopDelay=false
   )
   &SpawnBurst_Instantaneous(Spawn Count=8, Use Spawn Probability=false, UseLoopCountLimit=false)
  ]
  particleSpawnScript [
   &PartitionParticles(Number Of Partitions=8)
   &InitializeParticle(
    Color=&MultiplyLinearColor(
     Linear Color=<rgba 0.02 0.0057 0.0057 1.4>
     Scale Factor=&MakeVector4_FromLinearColor(LinearColor=&FNTimeOfDayColor(Night Color=<rgba 0.2448 0.2807 1 1>, Morning Color=<rgba 1 1 1 1>, Day Color=<rgba 1 1 1 1>, Evening Color=<rgba 1 1 1 1>))
    )
    Mesh Scale=&Multiply_Vector(
     A=&SelectVectorFromArray(
      Direct Array Index=$Particles.PartitionParticles.Partition
      Vector Selection Array=DataInterface() {propertyValues {FloatData [<1.1 1 1>, <1.1 1 1>, <-1.1 1 1>, <-1.1 1 1>, <1.1 1 1>, <1.1 1 1>, <-1.1 1 1>, <-1.1 1 0.9>]}}
     )
     B=&MakeCustomVectorFromBool(Bool=$User.bIsFrontEndPreview, True Vector=<0.6 0.8 0.6>, False Vector=<0.8 1 0.8>)
    )
    Mesh Scale Mode=::NewEnumerator4
    UsePositionOffset=false
    Write Material Random=true
   )
   &InitialMeshOrientation(
    Mesh Orientation Mode=::NewEnumerator2
    Rotation=&SelectVectorFromArray(
     Direct Array Index=$Particles.PartitionParticles.Partition
     Vector Selection Array=DataInterface() {propertyValues {FloatData [<0 0.05 0.15>, <0 0.1 -0.15>, <0 -0.05 0.15>, <0 -0.0859 -0.15>, <0 0.1 0.15>, <0 0.073 -0.15>, <0 -0.1 0.15>, <0 -0.0569 -0.15>]}}
    )
    Use Rotation Vector=true
   )
   &ShapeLocation(
    MaskSpawnGroup=false
    Offset=&MakeCustomVectorFromBool(
     Bool=$User.bIsFrontEndPreview
     True Vector=&SelectVectorFromArray(
      Direct Array Index=$Particles.PartitionParticles.Partition
      Vector Selection Array=DataInterface() {propertyValues {FloatData [<7 -6 6.5>, <6.8721 -18 6.5>, <-7 -18 6.5>, <-9 -6 6.5>, <5.2 -13 -5>, <5 -14 -5>, <-5 -14 -5>, <-6 -10 -5>]}}
     )
     False Vector=&SelectVectorFromArray(
      Direct Array Index=$Particles.PartitionParticles.Partition
      Vector Selection Array=DataInterface() {propertyValues {FloatData [<7 -7 6.8918>, <6 -16 7>, <-6 -18 7>, <-8 -7 8>, <7 -15 -2.9495>, <5.6 -13 -6>, <-5 -13 -5>, <-8.6524 -12 -6>]}}
     )
    )
    Offset Mode=::NewEnumerator0
    Override Seed=false
    Sphere Radius=0.001
    UseOwnerScale=true
   )
   &CameraOffset(Camera Offset Amount=-2)
  ]
  particleUpdateScript [
   &ParticleState(Kill Particles When Lifetime Has Elapsed=false, Let Infinitely Lived Particles Die When Emitter Deactivates=true)
   &ScaleColor(Scale Alpha=&RandomRangeFloat(Minimum=0.6, Maximum=1), ScaleA=true, ScaleRGB=true)
   /SetVariables(Particles.Color_Red:LinearColor=&MultiplyLinearColor(
     Linear Color=<rgba 0.0729 0.0148 0.0148 2>
     Scale Factor=&MakeVector4_FromFloat(Float=&FN_FrontEndFloatModifier(Gameplay=1, Front End=0.5))
    )) {isSetParameters=true, module=SetVariables_#4}
   &InterpolateOverTime(Rate Of Change=10, Target Vector Value:NiagaraPosition=$Particles.Position, Variable Type=::NewEnumerator2)
   /SetVariables(Particles.Interpolate Dynamic=&MakeFloat4FromVectorAndFloat("Vector (RGB)"=$Particles.InterpolateOverTime.InterpolateOverTime_V3_.Moving Average, "Float (Alpha)"=0)) {isSetParameters=true, module=SetVariables_#5}
   &DynamicMaterialParameters(
    Index 0 Param 1=1
    Index 0 Param 2=&RandomRangeFloat(Minimum=0.5, Maximum=1)
    Index 0 Param 3=&RandomRangeFloat(Minimum=0.5, Maximum=1)
    Index 0 Param 4=&FN_FrontEndFloatModifier(Gameplay=1, Front End=0.8)
    Index 2 Param 1=1
    Index 2 Param 2=1
    Index 2 Param 3=&RandomRangeFloat(Minimum=0.1, Maximum=0.2)
    Index 2 Param 4=&RandomRangeFloat(Minimum=0.2, Maximum=0.5)
    Module.Index0Param0Write=true
    Module.Index0Param1Write=true
    Module.Index0Param2Write=true
    Module.Index0Param3Write=true
    Module.Index2Param0Write=true
    Module.Index2Param1Write=true
    Module.Index2Param2Write=true
    Module.Index2Param3Write=true
    Write Parameter Index 2=true
   )
  ]
  rendererClasses [&NiagaraMeshRendererProperties]
  renderers [
   {rendererClass=&NiagaraMeshRendererProperties, rendererIndex=0}
   {rendererClass=&NiagaraMeshRendererProperties, rendererIndex=1}
  ]
 }
 Spline_horn {
  emitterUpdateScript [
   &EmitterState(
    Inactive Response=::NewEnumerator1
    Life Cycle Mode=::NewEnumerator1
    Loop Behavior=::NewEnumerator1
    Loop Duration Mode=::NewEnumerator1
    UseLoopDelay=false
   )
   &SpawnBurst_Instantaneous(Spawn Count=4, Use Spawn Probability=false, UseLoopCountLimit=false)
  ]
  particleSpawnScript [
   &PartitionParticles(Number Of Partitions=4)
   &InitializeParticle(
    Color=&MultiplyLinearColor(
     Linear Color=<rgba 0.01 0.0029 0.0029 1.4>
     Scale Factor=&MakeVector4_FromLinearColor(LinearColor=&FNTimeOfDayColor(Night Color=<rgba 0.4048 0.4331 1 1>, Morning Color=<rgba 1 1 1 1>, Day Color=<rgba 1 1 1 1>, Evening Color=<rgba 1 1 1 1>))
    )
    Mesh Scale=&Multiply_Vector(
     A=&SelectVectorFromArray(
      Direct Array Index=$Particles.PartitionParticles.Partition
      Vector Selection Array=DataInterface() {propertyValues {FloatData [<1.1 1 1.3>, <1.05 2 1.5>, <-1.2 1 1.3>, <-1 1 1.5>]}}
     )
     B=&MakeCustomVectorFromBool(Bool=$User.bIsFrontEndPreview, True Vector=<0.9 0.9 0.8>, False Vector=<1 1 1>)
    )
    Mesh Scale Mode=::NewEnumerator4
    UsePositionOffset=false
    Write Material Random=true
   )
   &InitialMeshOrientation(
    Mesh Orientation Mode=::NewEnumerator2
    Rotation=&SelectVectorFromArray(
     Direct Array Index=$Particles.PartitionParticles.Partition
     Vector Selection Array=DataInterface() {propertyValues {FloatData [<0 0 0>, <0 -0.0259 0>, <0 0.0198 0>, <0 0 0>]}}
    )
    Use Rotation Vector=true
   )
   &ShapeLocation(
    MaskSpawnGroup=false
    Offset=&MakeCustomVectorFromBool(
     Bool=$User.bIsFrontEndPreview
     True Vector=&SelectVectorFromArray(
      Direct Array Index=$Particles.PartitionParticles.Partition
      Vector Selection Array=DataInterface() {propertyValues {FloatData [<5 -13 7>, <0.746 -13 8.2767>, <-5 -13 7>, <0.746 -13 8.2767>]}}
     )
     False Vector=&SelectVectorFromArray(
      Direct Array Index=$Particles.PartitionParticles.Partition
      Vector Selection Array=DataInterface() {propertyValues {FloatData [<5 -14 7>, <0.746 -14 8.2767>, <-5 -14 7>, <0.746 -14 8.2767>]}}
     )
    )
    Offset Mode=::NewEnumerator0
    Override Seed=false
    Sphere Radius=0.001
    UseOwnerScale=true
   )
   &CameraOffset(Camera Offset Amount=-2)
  ]
  particleUpdateScript [
   &ParticleState(Kill Particles When Lifetime Has Elapsed=false, Let Infinitely Lived Particles Die When Emitter Deactivates=true)
   &ScaleColor(Scale Alpha=&RandomRangeFloat(Minimum=0.7, Maximum=1), ScaleA=true, ScaleRGB=true)
   /SetVariables(Particles.Color_Red:LinearColor=<rgba 0.2 0.0323 0.0436 1>) {isSetParameters=true, module=SetVariables_#4}
   &InterpolateOverTime(Rate Of Change=10, Target Vector Value:NiagaraPosition=$Particles.Position, Variable Type=::NewEnumerator2)
   /SetVariables(Particles.Interpolate Dynamic=&MakeFloat4FromVectorAndFloat("Vector (RGB)"=$Particles.InterpolateOverTime.InterpolateOverTime_V3_.Moving Average, "Float (Alpha)"=0)) {isSetParameters=true, module=SetVariables_#5}
   &DynamicMaterialParameters(
    Index 0 Param 1=1
    Index 0 Param 2=&RandomRangeFloat(Minimum=0.2, Maximum=0.5)
    Index 0 Param 3=&RandomRangeFloat(Minimum=0.2, Maximum=0.5)
    Index 2 Param 1=1
    Index 2 Param 2=1
    Index 2 Param 3=&RandomRangeFloat(Minimum=0.8, Maximum=1.2)
    Index 2 Param 4=&RandomRangeFloat(Minimum=0.8, Maximum=1)
    Module.Index0Param0Write=true
    Module.Index0Param1Write=true
    Module.Index0Param2Write=true
    Module.Index0Param3Write=true
    Module.Index2Param0Write=true
    Module.Index2Param1Write=true
    Module.Index2Param2Write=true
    Module.Index2Param3Write=true
    Write Parameter Index 2=true
   )
  ]
  rendererClasses [&NiagaraMeshRendererProperties]
  renderers [
   {rendererClass=&NiagaraMeshRendererProperties, rendererIndex=0}
   {rendererClass=&NiagaraMeshRendererProperties, rendererIndex=1}
  ]
 }
 Under {
  emitterUpdateScript [
   &EmitterState(
    Inactive Response=::NewEnumerator1
    Life Cycle Mode=::NewEnumerator1
    Loop Behavior=::NewEnumerator1
    Loop Duration Mode=::NewEnumerator1
    UseLoopDelay=false
   )
   &SpawnBurst_Instantaneous(Use Spawn Probability=false, UseLoopCountLimit=false)
  ]
  particleSpawnScript [
   &InitializeParticle(
    Color=&MultiplyLinearColor(
     Linear Color=<rgba 0.5 0.2604 0.2604 1.5>
     Scale Factor=&MakeVector4_FromLinearColor(LinearColor=&FNTimeOfDayColor(Night Color=<rgba 1 0.1846 0.1448 1>, Morning Color=<rgba 1 1 1 1>, Day Color=<rgba 1 1 1 1>, Evening Color=<rgba 1 1 1 1>))
    )
    Position Offset=<0 -12 -5>
    Sprite Size=<48 55>
    Sprite Size Mode=::NewEnumerator4
    UsePositionOffset=true
    Write Material Random=true
   )
   /SetVariables(Particles.Position_Red:NiagaraPosition=<0 -12 3>) {isSetParameters=true, module=SetVariables_#1}
   &CameraOffset()
   &SpriteFacingAndAlignment(Sprite Alignment=<0 0 1>, Sprite Facing=<0 0 1>, WriteAlignment=true, WriteFacing=true)
  ]
  particleUpdateScript [
   &ParticleState(Kill Particles When Lifetime Has Elapsed=false, Let Infinitely Lived Particles Die When Emitter Deactivates=true)
   /SetVariables(Particles.Color_Red:LinearColor=&MultiplyLinearColor(
     Linear Color=<rgba 0.1 0.0021 0.0021 5>
     Scale Factor=&MakeVector4_FromFloat(Float=&FN_FrontEndFloatModifier(Gameplay=1, Front End=0.7))
    )) {isSetParameters=true, module=SetVariables_#2}
   /SetVariables(Particles.Size_Red:Vector2f=<40 45>) {isSetParameters=true, module=SetVariables_#3}
  ]
  rendererClasses [&NiagaraSpriteRendererProperties]
  renderers [
   {rendererClass=&NiagaraSpriteRendererProperties}
   {rendererClass=&NiagaraSpriteRendererProperties, rendererIndex=1}
  ]
 }
}
summary {
 systemName=NS_Backpack_PigeonChart_Idle
 emitters [
  {emitterName=Under, rendererClasses [&NiagaraSpriteRendererProperties]}
  {emitterName=Spline_horn, rendererClasses [&NiagaraMeshRendererProperties]}
  {emitterName=Mid, rendererClasses [&NiagaraSpriteRendererProperties]}
  {emitterName=Spline_Edgeside, rendererClasses [&NiagaraMeshRendererProperties]}
 ]
 userVariables [
  User.bIsFrontEnd {type=NiagaraBool}
  User.bIsFrontEndPreview {type=NiagaraBool}
 ]
}
