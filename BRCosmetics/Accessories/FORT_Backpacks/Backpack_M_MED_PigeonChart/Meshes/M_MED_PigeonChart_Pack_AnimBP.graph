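# Animation Blueprint for the PigeonChart backpack cosmetic, set up for procedural motion of five stone bones, two left and three right.
# Each update, the EventGraph sets a vector per stone from real-time sine waves with differing frequencies and amplitudes, likely for an idle bobbing float.
# The AnimGraph chains additive bone-space Transform Bone nodes on those bones over the input pose, though their inputs show zero defaults with no visible link to the vectors.
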AnimBlueprint /BRCosmetics/Accessories/FORT_Backpacks/Backpack_M_MED_PigeonChart/Meshes/M_MED_PigeonChart_Pack_AnimBP
target_skeleton=&/M_MED_PigeonChart_Pack_Skeleton
graphs [
 AnimGraph {nodes [
   n1 = "Transform (Modify) Bone - Bone: dyn_stone_01_l"(ComponentPose=Local To Component(LocalPose=Input Pose()), Alpha=1, Translation="0,0,0", Rotation="0,0,0", Scale="1,1,1") {
    rotationMode=BMM_Additive
    rotationSpace=BCS_BoneSpace
    scaleMode=BMM_Additive
    scaleSpace=BCS_BoneSpace
    translationMode=BMM_Additive
    translationSpace=BCS_BoneSpace
    alphaBoolBlend {blendOption=Linear}
    boneToModify {boneName=dyn_stone_01_l}
   }
   n2 = "Transform (Modify) Bone - Bone: dyn_stone_02_l"(ComponentPose=@n1, Alpha=1, Translation="0,0,0", Rotation="0,0,0", Scale="1,1,1") {
    rotationMode=BMM_Additive
    rotationSpace=BCS_BoneSpace
    scaleMode=BMM_Additive
    scaleSpace=BCS_BoneSpace
    translationMode=BMM_Additive
    translationSpace=BCS_BoneSpace
    alphaBoolBlend {blendOption=Linear}
    boneToModify {boneName=dyn_stone_02_l}
   }
   n3 = "Transform (Modify) Bone - Bone: dyn_stone_01_r"(ComponentPose=@n2, Alpha=1, Translation="0,0,0", Rotation="0,0,0", Scale="1,1,1") {
    rotationMode=BMM_Additive
    rotationSpace=BCS_BoneSpace
    scaleMode=BMM_Additive
    scaleSpace=BCS_BoneSpace
    translationMode=BMM_Additive
    translationSpace=BCS_BoneSpace
    alphaBoolBlend {blendOption=Linear}
    boneToModify {boneName=dyn_stone_01_r}
   }
   n4 = "Transform (Modify) Bone - Bone: dyn_stone_02_r"(ComponentPose=@n3, Alpha=1, Translation="0,0,0", Rotation="0,0,0", Scale="1,1,1") {
    rotationMode=BMM_Additive
    rotationSpace=BCS_BoneSpace
    scaleMode=BMM_Additive
    scaleSpace=BCS_BoneSpace
    translationMode=BMM_Additive
    translationSpace=BCS_BoneSpace
    alphaBoolBlend {blendOption=Linear}
    boneToModify {boneName=dyn_stone_02_r}
   }
   n5 = "Transform (Modify) Bone - Bone: dyn_stone_03_r"(ComponentPose=@n4, Alpha=1, Translation="0,0,0", Rotation="0,0,0", Scale="1,1,1") {
    rotationMode=BMM_Additive
    rotationSpace=BCS_BoneSpace
    scaleMode=BMM_Additive
    scaleSpace=BCS_BoneSpace
    translationMode=BMM_Additive
    translationSpace=BCS_BoneSpace
    alphaBoolBlend {blendOption=Linear}
    boneToModify {boneName=dyn_stone_03_r}
   }
   Output Pose(Result=Component To Local(ComponentPose=@n5)) {layerGroup=None}
  ]}
 EventGraph {nodes [
   Event BlueprintUpdateAnimation()
   > n1 = Sequence()
   n2 = GetRealTimeSeconds()
   n3 = "Sin (Radians)"(A=float * float(A=2.8, B=@n2))
   n4 = "Sin (Radians)"(A=float * float(A=0.7, B=@n2))
   n5 = "Sin (Radians)"(A=float * float(A=1, B=@n2))
   > Set dyn_Stone_01_R(dyn_Stone_01_R:Vector=MakeVector(X=float * float(A=@n5, B=0.5), Y=float * float(A=@n4, B=0.7), Z=float * float(A=@n3, B=0.8))) -> (Output_Get:Vector)
   n6 = GetRealTimeSeconds()
   n7 = "Sin (Radians)"(A=float * float(A=0.7, B=@n6))
   n8 = "Sin (Radians)"(A=float * float(A=-1.5, B=@n6))
   n9 = "Sin (Radians)"(A=float * float(A=1.7, B=@n6))
   > Set dyn_Stone_02_R(dyn_Stone_02_R:Vector=MakeVector(X=float * float(A=@n8, B=0.7), Y=float * float(A=@n7, B=0.7), Z=float * float(A=@n9, B=0.7))) -> (Output_Get:Vector)
   n10 = GetRealTimeSeconds()
   n11 = "Sin (Radians)"(A=float * float(A=2, B=@n10))
   n12 = "Sin (Radians)"(A=float * float(A=1.5, B=@n10))
   n13 = "Sin (Radians)"(A=float * float(A=0.7, B=@n10))
   > Set dyn_Stone_03_R(dyn_Stone_03_R:Vector=MakeVector(X=float * float(A=@n12, B=0.5), Y=float * float(A=@n13, B=0.7), Z=float * float(A=@n11, B=0.5))) -> (Output_Get:Vector)
   n14 = GetRealTimeSeconds()
   n15 = float * float(A="Sin (Radians)"(A=float * float(A=1.5, B=@n14)), B=0.5)
   n16 = float * float(A="Sin (Radians)"(A=float * float(A=-2, B=@n14)), B=0.6)
   n17 = float * float(A="Sin (Radians)"(A=float * float(A=0.8, B=@n14)), B=0.8)
   Set dyn_Stone_01_L(execute=@n1.then_1, dyn_Stone_01_L:Vector=vector + vector(A=MakeVector(X=@n15, Y=@n17, Z=@n16), B="0.5,0,0")) -> (Output_Get:Vector)
   n18 = GetRealTimeSeconds()
   n19 = "Sin (Radians)"(A=float * float(A=1.5, B=@n18))
   n20 = "Sin (Radians)"(A=float * float(A=3, B=@n18))
   n21 = "Sin (Radians)"(A=float * float(A=1.2, B=@n18))
   > Set dyn_Stone_02_L(dyn_Stone_02_L:Vector=MakeVector(X=float * float(A=@n19, B=0.5), Y=float * float(A=@n21, B=1), Z=float * float(A=@n20, B=0.5))) -> (Output_Get:Vector)
  ]}
]
variables [
 dyn_Stone_01_L {type=Vector}
 dyn_Stone_01_R {type=Vector}
 dyn_Stone_02_L {type=Vector}
 dyn_Stone_02_R {type=Vector}
 dyn_Stone_03_R {type=Vector}
]
